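# Unlit, two-sided Media Plate material that cross-fades up to four media textures.
# A Custom HLSL node sums MediaTexture and MediaTexture1, weighted by Blend0 and Blend1, then layers that pair over the similarly weighted MediaTexture2 and MediaTexture3 using the first pair's alpha.
# Dynamic branches skip sampling for zero-weight textures and skip compositing when the lower pair is unused.
# The float4 result is broken out into separate components.
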Material /MediaPlate/M_MediaPlateCrossFade
graph {
 properties {
  bCastRayTracedShadows=false
  bHasPixelAnimation=true
  bOutputTranslucentVelocity=true
  pixelDepthOffsetMode=PDOM_Legacy
  shadingModel=MSM_Unlit
  twoSided=true
  usedShadingModels=Unlit
 }
 nodes [
  MaterialExpressionTextureObjectParameter_0 = TextureObjectParameter() {parameterName=MediaTexture, texture=&/T_MediaPlate}
  MaterialExpressionScalarParameter_1 = ScalarParameter() {parameterName=Blend0}
  MaterialExpressionTextureObjectParameter_1 = TextureObjectParameter() {parameterName=MediaTexture1, texture=&T_MediaPlate}
  MaterialExpressionScalarParameter_0 = ScalarParameter() {parameterName=Blend1}
  MaterialExpressionTextureObjectParameter_2 = TextureObjectParameter() {parameterName=MediaTexture2, texture=&T_MediaPlate}
  MaterialExpressionScalarParameter_3 = ScalarParameter() {parameterName=Blend2}
  MaterialExpressionTextureObjectParameter_3 = TextureObjectParameter() {parameterName=MediaTexture3, texture=&T_MediaPlate}
  MaterialExpressionScalarParameter_2 = ScalarParameter() {parameterName=Blend3}
  MaterialExpressionTextureCoordinate_0 = TextureCoordinate()
  n1 = Custom!(
   Tex0=@MaterialExpressionTextureObjectParameter_0
   Blend0=@MaterialExpressionScalarParameter_1
   Tex1=@MaterialExpressionTextureObjectParameter_1
   Blend1=@MaterialExpressionScalarParameter_0
   Tex2=@MaterialExpressionTextureObjectParameter_2
   Blend2=@MaterialExpressionScalarParameter_3
   Tex3=@MaterialExpressionTextureObjectParameter_3
   Blend3=@MaterialExpressionScalarParameter_2
   UV=@MaterialExpressionTextureCoordinate_0
  ) {
   code="float4 Output = float4(0,0,0,0);\r\nBRANCH\r\nif (Blend0 > 0.0)\r\n{\r\n\tOutput = Texture2DSample(Tex0, Tex0Sampler, UV) * Blend0;\r\n}\r\nBRANCH\r\nif (Blend1 > 0.0)\r\n{\r\n\tOutput += Texture2DSample(Tex1, Tex1Sampler, UV) * Blend1;\r\n}\r\nBRANCH\r\nif (Blend2 + Blend3 > 0.0)\r\n{\r\n\tfloat4 Output1 = float4(0,0,0,0);\r\n\tBRANCH\r\n\tif (Blend2 > 0.0)\r\n\t{\r\n\t\tOutput1 = Texture2DSample(Tex2, Tex2Sampler, UV) * Blend2;\r\n\t}\r\n\tBRANCH\r\n\tif (Blend3 > 0.0)\r\n\t{\r\n\t\tOutput1 += Texture2DSample(Tex3, Tex3Sampler, UV) * Blend3;\r\n\t}\r\n    Output.rgb = lerp(Output1.rgb, Output.rgb, Output.a);\r\n\tOutput.a = Output1.a * (1.0f - Output.a) + Output.a;\r\n}\r\nreturn Output;"
   outputType=CMOT_Float4
   inputs [
    {inputName=Tex0, input {expression=&self:MaterialExpressionTextureObjectParameter_0}}
    {inputName=Blend0, input {expression=&self:MaterialExpressionScalarParameter_1}}
    {inputName=Tex1, input {expression=&self:MaterialExpressionTextureObjectParameter_1}}
    {inputName=Blend1, input {expression=&self:MaterialExpressionScalarParameter_0}}
    {inputName=Tex2, input {expression=&self:MaterialExpressionTextureObjectParameter_2}}
    {inputName=Blend2, input {expression=&self:MaterialExpressionScalarParameter_3}}
    {inputName=Tex3, input {expression=&self:MaterialExpressionTextureObjectParameter_3}}
    {inputName=Blend3, input {expression=&self:MaterialExpressionScalarParameter_2}}
    {inputName=UV, input {expression=&self:MaterialExpressionTextureCoordinate_0}}
   ]
  }
  MaterialFunctionCall(Float4=@n1) {materialFunction=&BreakOutFloat4Components}
 ]
}
