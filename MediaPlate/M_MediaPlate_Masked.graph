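# Unlit, two-sided masked material for a Media Plate, displaying a media texture parameter with pixel animation flagged and ray traced shadows off.
# The texture's RGB passes through a static debug switch that can substitute solid red for the color.
# Its alpha runs through DitherTemporalAA, giving a dithered approximation of soft transparency within masked blending.
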Material /MediaPlate/M_MediaPlate_Masked
graph {
 properties {
  bCastRayTracedShadows=false
  bHasPixelAnimation=true
  bOutputTranslucentVelocity=true
  blendMode=BLEND_Masked
  pixelDepthOffsetMode=PDOM_Legacy
  shadingModel=MSM_Unlit
  translucencyPass=MTP_BeforeDOF
  twoSided=true
  usedShadingModels=Unlit
 }
 nodes [
  n1 = TextureSampleParameter2D() {parameterName=MediaTexture, texture=&/T_MediaPlate}
  MaterialFunctionCall(Alpha Threshold=@n1.A) {materialFunction=&DitherTemporalAA}
  StaticSwitch(True=Constant3Vector() {constant=<rgba 1 0 0 0>}, False=@n1.RGB, Value=StaticBool() {desc="Debug  Mode"})
 ]
}
