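# Opaque, two-sided unlit material for media plates, exposing a texture slot for media content.
# Its single node is a 2D texture sample parameter named MediaTexture, defaulting to the T_MediaPlate texture.
# It is flagged as having pixel animation, suiting changing imagery, with translucent velocity output enabled and ray-traced shadow casting disabled.
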Material /MediaPlate/M_MediaPlate_Opaque
graph {
 properties {
  bCastRayTracedShadows=false
  bHasPixelAnimation=true
  bOutputTranslucentVelocity=true
  pixelDepthOffsetMode=PDOM_Legacy
  shadingModel=MSM_Unlit
  twoSided=true
  usedShadingModels=Unlit
 }
 nodes [TextureSampleParameter2D() {parameterName=MediaTexture, texture=&/T_MediaPlate}]
}
