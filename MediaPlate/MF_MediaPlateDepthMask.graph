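# Material function for Media Plate that builds a binary depth mask by comparing scene depth with the current pixel's depth.
# It samples scene depth at the screen pixel position plus an optional PixelOffset, normalized by the visible resolution.
# A Step on the absolute depth difference outputs 1 where it lies within DepthDistanceClip and 0 where it exceeds that distance.
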MaterialFunction /MediaPlate/MF_MediaPlateDepthMask
graph {
 properties {previewBlendMode=BLEND_Translucent}
 nodes [
  n1 = Add(
   A=ScreenPosition().PixelPosition
   B=FunctionInput(Preview=Constant2Vector()) {bUsePreviewValueAsDefault=true, inputName=PixelOffset, inputType=FunctionInput_Vector2, sortPriority=1}
  )
  n2 = MaterialFunctionCall() {materialFunction=&ScreenResolution}
  n3 = Subtract(A=SceneDepth(UVs=Divide(A=@n1, B=@n2.Visible Resolution)).R, B=PixelDepth().R)
  n4 = FunctionInput() {bUsePreviewValueAsDefault=true, inputName=DepthDistanceClip, inputType=FunctionInput_Scalar, sortPriority=2, previewValue=<5 0 0 1>}
  FunctionOutput(None=Step(Y=Abs(None=@n3), X=@n4))
 ]
}
