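# Two-sided unlit translucent Media Plate overlay composite material, rendered after motion blur, that masks the media texture's alpha by scene depth.
# Eight MF_MediaPlateDepthMask samples at axis and diagonal pixel offsets multiply together, then with that alpha, so any failing neighbor cuts the mask.
# The offset scales FilterOffset inversely with resolution fraction, within a clamp.
# DepthMaskOffset sets the depth clip distance, widening up to tenfold at grazing view angles.
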Material /MediaPlate/M_MediaPlate_OverlayComp
graph {
 properties {
  blendMode=BLEND_Translucent
  pixelDepthOffsetMode=PDOM_Legacy
  shadingModel=MSM_Unlit
  translucencyPass=MTP_AfterMotionBlur
  twoSided=true
  usedShadingModels=Unlit
 }
 nodes [
  n1 = Clamp(None=Divide(A=ScalarParameter() {defaultValue=1, parameterName=FilterOffset}, B=ViewProperty() {property=MEVP_ResolutionFraction}.Property)) {maxDefault=3, minDefault=1}
  n2 = ScalarParameter() {defaultValue=5, parameterName=DepthMaskOffset}
  n3 = Abs(None=DotProduct(
    A=PixelNormalWS()
    B=MaterialFunctionCall() {materialFunction=&CameraDirectionVector}.Result
   ))
  n4 = LinearInterpolate(A=@n2, B=Multiply(A=@n2) {constB=10}, Alpha=OneMinus(None=Saturate(None=@n3)))
  n5 = Multiply(A=@n1) {constB=-1}
  n6 = Multiply(
   A=MaterialFunctionCall!(
    PixelOffset=MaterialFunctionCall(X=@n1) {materialFunction=&MakeFloat2}.Result
    DepthDistanceClip=@n4
   ) {materialFunction=&/MF_MediaPlateDepthMask}.Result
   B=MaterialFunctionCall(
    PixelOffset=MaterialFunctionCall(X=@n5) {materialFunction=&MakeFloat2}.Result
    DepthDistanceClip=@n4
   ) {materialFunction=&MF_MediaPlateDepthMask}.Result
  )
  n7 = Multiply(
   A=MaterialFunctionCall(
    PixelOffset=MaterialFunctionCall(Y=@n1) {materialFunction=&MakeFloat2}.Result
    DepthDistanceClip=@n4
   ) {materialFunction=&MF_MediaPlateDepthMask}.Result
   B=MaterialFunctionCall(
    PixelOffset=MaterialFunctionCall(Y=@n5) {materialFunction=&MakeFloat2}.Result
    DepthDistanceClip=@n4
   ) {materialFunction=&MF_MediaPlateDepthMask}.Result
  )
  n8 = Multiply(A=@n1) {constB=-1}
  n9 = Multiply(
   A=MaterialFunctionCall(
    PixelOffset=MaterialFunctionCall(X=@n1, Y=@n1) {materialFunction=&MakeFloat2}.Result
    DepthDistanceClip=@n4
   ) {materialFunction=&MF_MediaPlateDepthMask}.Result
   B=MaterialFunctionCall(
    PixelOffset=MaterialFunctionCall(X=@n1, Y=@n8) {materialFunction=&MakeFloat2}.Result
    DepthDistanceClip=@n4
   ) {materialFunction=&MF_MediaPlateDepthMask}.Result
  )
  n10 = Multiply(
   A=MaterialFunctionCall(
    PixelOffset=MaterialFunctionCall(X=@n8, Y=@n1) {materialFunction=&MakeFloat2}.Result
    DepthDistanceClip=@n4
   ) {materialFunction=&MF_MediaPlateDepthMask}.Result
   B=MaterialFunctionCall(
    PixelOffset=MaterialFunctionCall(X=@n8, Y=@n8) {materialFunction=&MakeFloat2}.Result
    DepthDistanceClip=@n4
   ) {materialFunction=&MF_MediaPlateDepthMask}.Result
  )
  Multiply(
   A=TextureSampleParameter2D() {parameterName=MediaTexture, texture=&/T_MediaPlate}.A
   B=Multiply(A=Multiply(A=@n6, B=@n7), B=Multiply(A=@n9, B=@n10))
  )
 ]
}
